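# Niagara system for the FrostChill shoes' staged cold effects, with emitters sampling skeletal mesh sockets or bones on both feet.
# Stage 2 drives timeline-ramped frost sprites; entering stage 3 fires a one-shot mesh and sprite ice burst.
# Running in stage 3 adds ribbon trails, ice flecks and smoke, suppressed by skydiving or HideFootFX.
# Per-foot shoe scale sizes particles, time of day tints or dims them, and front-end flags adjust preview behavior.
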NiagaraSystem /CosmeticShoes/Assets/FORT_Shoes/Shoe_FrostChill/FX/NS_Shoe_FrostChill_Idle
dependencies {
 usedDataInterfaces [
  &NiagaraDataInterfaceArrayInt32
  &NiagaraDataInterfaceCurve
  &NiagaraDataInterfaceSpriteRendererInfo
  &NiagaraDataInterfaceColorCurve
 ]
 usedDynamicInputs [
  &SetBoolByFloatComparison
  &MakeFloatFromInt
  &Multiply_Vector2D_ByFloat
  &RandomRangeVector2D
  /Game/Effects/Niagara/DynamicInputs/Logic/FN_IsRunningCheck
  &Multiply_Vector
  &ModuloFloat
  &Divide_Float
 ]
 usedModules [
  &self:SystemSpawnScript.SystemScriptSource.SystemScriptGraph.NiagaraNodeAssignment_0.SetVariables_#1
  &self:SystemSpawnScript.SystemScriptSource.SystemScriptGraph.NiagaraNodeAssignment_4.SetVariables_#2
  &SystemState
  &self:SystemSpawnScript.SystemScriptSource.SystemScriptGraph.NiagaraNodeAssignment_1.SetVariables_#3
  &self:SystemSpawnScript.SystemScriptSource.SystemScriptGraph.NiagaraNodeAssignment_2.SetVariables_#4
  &Timeline
  &self:SystemSpawnScript.SystemScriptSource.SystemScriptGraph.NiagaraNodeAssignment_3.SetVariables_#5
 ]
}
emitters {
 FN_AttachedRibbon001 {
  emitterUpdateScript [
   &EmitterState()
   &FN_SpawnRibbon_v2(
    Number of Ribbons:NiagaraInt32=&MakeCustomIntFromBool(
     Bool=&Boolean_OrOperation(Bool1=$User.IsSkydiving, Bool2=&Boolean_NotOperation(Bool=$User.isStage3))
     True INT=0
     False INT=2
    )
    Spawn Mode=::NewEnumerator0
    SpawnRate:NiagaraFloat=&MakeCustomFloatFromBool(
     Bool=$User.IsSkydiving
     True Float=0
     False Float=&MakeCustomFloatFromBool(
      Bool=$User.HideFootFX
      True Float=0
      False Float=&MakeCustomFloatFromBool(Bool=$System.isRunning, True Float=12, False Float=0)
     )
    )
   )
  ]
  particleSpawnScript [
   /Game/Effects/Niagara/Modules/Ribbons/FN_InitializeRibbon_v4(
    Enable Skeletal Position=true
    Lifetime:NiagaraFloat=&FN_FrontEndFloatModifier(Gameplay=0.17, Front End=0.45)
    Ribbon Width=7
    SkeletalMesh:NiagaraDataInterfaceSkeletalMesh=$System.OnlyRibbons
   )
   &FN_FlipRibbonV()
   &DynamicMaterialParameters(
    Index 0 Param 1=1
    Index 0 Param 2=0.75
    Index 0 Param 3=1
    Index 0 Param 4=1
    Module.Index0Param0Write=true
    Module.Index0Param1Write=true
    Module.Index0Param2Write=true
    Module.Index0Param3Write=true
   )
   /SetVariables(Particles.RibbonWidth=&Multiply_Float(
     A=&MergeHelper_FloatSwitch(
      0=&Multiply_Float(A=&MakeFloatFromVector(Vector=$User.ShoeScale_L, Channel=::NewEnumerator0), B=1)
      1=&Multiply_Float(A=&MakeFloatFromVector(Vector=$User.ShoeScale_R, Channel=::NewEnumerator0), B=1)
      2=?
      3=?
      Index=&SpawnGroup()
      Num=::NewEnumerator0
     )
     B=&FloatFromCurve(
      FloatCurve=DataInterface() {propertyValues {Curve {keys [
          {arriveTangent=-0.3053, interpMode=RCIM_Cubic, leaveTangent=-0.3053, tangentMode=RCTM_User, time=0, value=1}
          {arriveTangent=-0.5499, interpMode=RCIM_Cubic, leaveTangent=-0.5499, tangentMode=RCTM_User, time=1, value=0.5}
         ]}}}
      Scale Curve=20
     )
    )) {isSetParameters=true}
   &AddVelocity(
    Additional Quaternion Rotation=true
    Additional Yaw / Pitch / Roll:Vector3f=&MaskVectorBySpawnGroup(Masked Vector=<-80 0 -10>, Passthrough Vector=<100 0 -10>)
    Rotation Mode=::NewEnumerator3
    Rotation Quaternion:Quat4f=$Particles.SampledRotation
    Velocity=<0 -200 50>
    Velocity Speed Scale=&FN_FrontEndFloatModifier(Gameplay=0, Front End=1)
   ) {module=AddVelocity001}
  ]
  particleUpdateScript [
   &InheritSourceMovement(
    Applied Position Scale=&VectorFromFloat(Value=&FloatFromCurve(FloatCurve=DataInterface() {propertyValues {Curve {keys [
          {arriveTangent=-0.2727, interpMode=RCIM_Cubic, leaveTangent=-0.2727, tangentMode=RCTM_User, time=0, value=1}
          {arriveTangent=-0.5057, interpMode=RCIM_Cubic, leaveTangent=-0.5057, tangentMode=RCTM_User, time=1, value=0.14}
         ]}}}))
    ApplyForce=false
    ApplyPosition=true
    ApplyPositionFalloff=false
   )
   &SolveForcesAndVelocity(Clamp Velocity=false, Limit Acceleration=false)
   /Game/Effects/Niagara/Modules/Ribbons/FN_UpdateRibbon_v4(Enable Skeletal Position=true, SkeletalMesh:NiagaraDataInterfaceSkeletalMesh=$System.OnlyRibbons)
   /SetVariables(Particles.AlphaScale:NiagaraFloat=&FN_FrontEndFloatModifier(
     Gameplay=&RemapRange(
      Input Value To Remap=&VectorLength(Vector=$Particles.FN_UpdateRibbon_v4.SampledVelocity)
      Input Min=400
      Input Max=1000
      Output Min=0
      Output Max=1
     )
     Front End=1
    )) {isSetParameters=true}
   &FN_RibbonMaterial(
    Color=<rgba 0.1944 0.5954 1.566 1>
    Color Pan U Speed=0.1
    Opacity Pan U Speed=1.5
    Opacity Power=0.5
    Opacity Scale U=2
    Scale Alpha:NiagaraFloat=&FloatFromCurve(
     FloatCurve=DataInterface() {propertyValues {Curve {keys [
         {arriveTangent=-1.0872, interpMode=RCIM_Cubic, leaveTangent=-1.0872, tangentMode=RCTM_User, time=0, value=0.5}
         {arriveTangent=-0.3496, interpMode=RCIM_Cubic, leaveTangent=-0.3496, tangentMode=RCTM_User, time=1, value=0}
        ]}}}
     Scale Curve=&Multiply_Float(
      A=$Particles.AlphaScale
      B=&Multiply_Float(A=$User.TrailOn, B=&FNTimeOfDayScalar(Night Scale=0.6, Morning Scale=1, Evening Scale=1))
     )
    )
   )
  ]
  rendererClasses [&NiagaraRibbonRendererProperties]
  renderers [
   {rendererClass=&NiagaraRibbonRendererProperties}
   {rendererClass=&NiagaraRibbonRendererProperties, rendererIndex=1}
  ]
 }
 Frost {
  emitterUpdateScript [
   &EmitterState()
   &SpawnRate(
    Spawn Probability=0.6
    SpawnRate=&MakeCustomFloatFromBool(
     Bool=$User.isOffPawnPreview
     True Float=$System.FELoop
     False Float=&MakeCustomFloatFromBool(
      Bool=$System.isFalling
      True Float=0
      False Float=&MakeCustomFloatFromBool(
       Bool=$User.isStage2
       True Float=&Multiply_Float(A=&Lerp_Float(A=17, B=2, Alpha=$System.Timeline001.Time), B=$System.Timeline.Time)
       False Float=0
      )
     )
    )
    Use Spawn Probability=true
   )
  ]
  particleSpawnScript [
   &PartitionParticles()
   &InitializeParticle(
    Color=&FNTimeOfDayColor(
     Night Color=<rgba 0.3082 0.4822 1 0.05>
     Morning Color=<rgba 0.271 0.4832 0.8003 0.06>
     Day Color=<rgba 0.3029 0.4445 0.7425 0.06>
     Evening Color=<rgba 0.271 0.4832 0.8003 0.06>
    )
    Lifetime Max=1.5
    Lifetime Min=1.2
    Lifetime Mode=::NewEnumerator1
    Sprite Rotation Mode=::NewEnumerator2
    Sprite Size Mode=::NewEnumerator3
    Uniform Sprite Size=&Multiply_Float(
     A=&RandomRangeFloat(Minimum=25, Maximum=15)
     B=&MakeCustomFloatFromBool(Bool=$User.isOffPawnPreview, True Float=0.5, False Float=1)
    )
    UsePositionOffset=false
    Write Material Random=true
   )
   &CameraOffset(Camera Offset Amount=&MakeCustomFloatFromBool(Bool=$User.isOffPawnPreview, True Float=20, False Float=10))
  ]
  particleUpdateScript [
   &ParticleState()
   &SkeletalMeshLocation(
    Bone / Socket Index=/CascadeToNiagaraConverter/NiagaraScripts/DynamicInputScripts/SequentialInt(
     Use Array Data:NiagaraBool=true
     First Number:NiagaraInt32=?
     Sequence:NiagaraDataInterfaceArrayInt32=DataInterface() {propertyValues {IntData [1,3,11,13,15]}}
     Last Number:NiagaraInt32=?
    )
    "Inherit Velocity (Scale)":NiagaraFloat=&FloatFromCurve(
     FloatCurve=DataInterface() {propertyValues {Curve {defaultValue=0, keys [{tangentMode=RCTM_User, time=0, value=0}, {tangentMode=RCTM_User, time=1, value=-1}]}}}
     Scale Curve=0.05
    )
    MaskSpawnGroup=false
    Mesh Sampling Type=::NewEnumerator5
    Skeletal Mesh=$System.Skeletal Mesh
    Socket Sampling Mode=::NewEnumerator5
    Velocity Sampling=::NewEnumerator0
    VelocitySpeedLimit=false
   )
   &SkeletalMeshLocation(
    Bone / Socket Index=&SequentialInt(
     Use Array Data:NiagaraBool=true
     First Number:NiagaraInt32=?
     Sequence:NiagaraDataInterfaceArrayInt32=DataInterface() {propertyValues {IntData [0,2,10,12,14]}}
     Last Number:NiagaraInt32=?
    )
    "Inherit Velocity (Scale)":NiagaraFloat=&FloatFromCurve(
     FloatCurve=DataInterface() {propertyValues {Curve {defaultValue=0, keys [{tangentMode=RCTM_User, time=0, value=0}, {tangentMode=RCTM_User, time=1, value=-1}]}}}
     Scale Curve=0.05
    )
    MaskSpawnGroup=false
    Mesh Sampling Type=::NewEnumerator5
    Skeletal Mesh=$System.Skeletal Mesh
    Socket Sampling Mode=::NewEnumerator5
    Velocity Sampling=::NewEnumerator0
    VelocitySpeedLimit=false
   ) {module=SkeletalMeshLocation001}
   /SetVariables(Particles.Position=&MergeHelper_VectorSwitch(
     0=Linked() {linkedVariable {name=Particles.SkeletalMeshLocation.SampledPosition, type=NiagaraPosition}}
     1=Linked() {linkedVariable {name=Particles.SkeletalMeshLocation001.SampledPosition, type=NiagaraPosition}}
     2=?
     3=?
     Index=$Particles.PartitionParticles.Partition
     Num=::NewEnumerator0
    )) {isSetParameters=true}
   &ScaleSpriteSize(
    Scale Sprite Size Mode=::NewEnumerator0
    Uniform Scale Factor=&Multiply_Float(
     A=&MergeHelper_FloatSwitch(
      0=&Multiply_Float(A=&MakeFloatFromVector(Vector=$User.ShoeScale_L, Channel=::NewEnumerator0), B=1)
      1=&Multiply_Float(A=&MakeFloatFromVector(Vector=$User.ShoeScale_R, Channel=::NewEnumerator0), B=1)
      2=?
      3=?
      Index=$Particles.PartitionParticles.Partition
      Num=::NewEnumerator0
     )
     B=&FloatFromCurve(FloatCurve=DataInterface() {propertyValues {Curve {
         defaultValue=0
         keys [
          {arriveTangent=3.2023, interpMode=RCIM_Cubic, leaveTangent=3.2023, tangentMode=RCTM_User, time=0, value=0}
          {interpMode=RCIM_Cubic, tangentMode=RCTM_User, time=1, value=1}
         ]
        }}})
    )
   )
   &ScaleColor(
    Scale Alpha=&MakeCustomFloatFromBool(Bool=$User.HideFootFX, True Float=0, False Float=1)
    ScaleA=true
    ScaleRGB=false
   ) {module=ScaleColor001}
   &SubUVAnimation(
    End Frame Range Override=10
    "Sprite Renderer "=DataInterface() {propertyValues {SpriteRenderer=&self:Frost_0.NiagaraSpriteRendererProperties_1}}
    UseEndFrame=true
    UseStartFrame=false
   )
   &ScaleColor(
    Scale Alpha=&FloatFromCurve(FloatCurve=DataInterface() {propertyValues {Curve {
        defaultValue=0
        keys [
         {interpMode=RCIM_Cubic, tangentMode=RCTM_User, time=0, value=0}
         {interpMode=RCIM_Cubic, tangentMode=RCTM_User, time=0.5, value=1}
         {tangentMode=RCTM_User, time=1, value=0}
        ]
       }}})
    ScaleA=true
    ScaleRGB=false
   )
   &DynamicMaterialParameters(
    Index 0 Param 1=3
    Index 0 Param 2=1
    Index 0 Param 3=0.5
    Module.Index0Float0Write=true
    Module.Index0Float1Write=true
    Module.Index0Float2Write=true
    Module.Index0Float3Write=true
   )
   &SolveForcesAndVelocity(Clamp Velocity=false, Limit Acceleration=false)
  ]
  rendererClasses [&NiagaraSpriteRendererProperties]
  renderers [{rendererClass=&NiagaraSpriteRendererProperties}]
 }
 Ice {
  emitterUpdateScript [
   &EmitterState()
   &SpawnRate(
    Spawn Probability=&MakeCustomFloatFromBool(Bool=$User.IsSkydiving, True Float=0, False Float=1)
    SpawnRate=&MakeCustomFloatFromBool(
     Bool=&Boolean_OrOperation(Bool1=$User.IsSkydiving, Bool2=&Boolean_NotOperation(Bool=$User.isStage3))
     True Float=0
     False Float=&MakeCustomFloatFromBool(
      Bool=$User.HideFootFX
      True Float=0
      False Float=&MakeCustomFloatFromBool(Bool=$System.isRunning, True Float=50, False Float=0)
     )
    )
    Use Spawn Probability=true
   )
  ]
  particleSpawnScript [
   &InitializeParticle(
    Color Mode=::NewEnumerator0
    Lifetime Max=0.1
    Lifetime Min=0.3
    Lifetime Mode=::NewEnumerator1
    Sprite Size Mode=::NewEnumerator1
    Uniform Sprite Size Max=0.5
    Uniform Sprite Size Min=1
    UsePositionOffset=false
    Write Material Random=true
   )
   &FN_SetSpawnGroup(Number of Spawn Groups=2)
   &SkeletalMeshLocation(Bone / Socket Index=$Particles.SpawnGroup, Bone Sampling Mode=::NewEnumerator4, MaskSpawnGroup=false, Skeletal Mesh=$System.Skeletal Mesh) {module=SkeletalMeshLocation001}
   &ShapeLocation(MaskSpawnGroup=false, Override Seed=false, Sphere Radius=10, UseOwnerScale=true)
   /SetVariables(Particles.InitSpriteRotRate=&UniformRangedFloat(Minimum=-1, Maximum=1)) {isSetParameters=true}
   &AddVelocity(
    Additional Quaternion Rotation=true
    Additional Yaw / Pitch / Roll:Vector3f=&MaskVectorBySpawnGroup(Masked Vector=<0 0 180>, Passthrough Vector=<0 0 0>)
    Rotation Mode=::NewEnumerator3
    Rotation Quaternion:Quat4f=$Particles.SkeletalMeshLocation001.SampledRotation
    Velocity=<0 120 30>
    Velocity Speed Scale=&FN_FrontEndFloatModifier(Gameplay=0, Front End=1)
   ) {module=AddVelocity001}
   &ApplyInitialForces()
  ]
  particleUpdateScript [
   &ParticleState()
   &DynamicMaterialParameters(
    Index 0 Param 1=&FloatFromCurve(
     FloatCurve=DataInterface() {propertyValues {Curve {keys [
         {arriveTangent=-5.8452, interpMode=RCIM_Cubic, time=0, value=1}
         {arriveTangentWeight=0.0434, interpMode=RCIM_Cubic, leaveTangentWeight=0.217, tangentMode=RCTM_User, time=0.2, value=0}
         {arriveTangent=3.4164, interpMode=RCIM_Cubic, leaveTangent=3.4164, tangentMode=RCTM_User, time=0.8, value=0.7}
         <at 1 0 RCIM_Cubic>
        ]}}}
     Scale Curve=3
    )
    Index 0 Param 3=1
    Index 0 Param 4=&FloatFromCurve(
     FloatCurve=DataInterface() {propertyValues {Curve {keys [
         {interpMode=RCIM_Cubic, leaveTangent=-1.5195, tangentMode=RCTM_Break, time=0, value=1}
         {arriveTangent=0.0067, interpMode=RCIM_Cubic, tangentMode=RCTM_Break, time=1, value=0}
        ]}}}
     Scale Curve=&RandomRangeFloat(Minimum=10, Maximum=30)
    )
    Module.Param0WriteEnabled=true
    Module.Param1WriteEnabled=true
    Module.Param2WriteEnabled=true
    Module.Param3WriteEnabled=true
   )
   &ScaleSpriteSize(
    Scale Sprite Size Mode=::NewEnumerator0
    Uniform Scale Factor=&Multiply_Float(
     A=&MergeHelper_FloatSwitch(
      0=&Multiply_Float(A=&MakeFloatFromVector(Vector=$User.ShoeScale_R, Channel=::NewEnumerator0), B=1)
      1=&Multiply_Float(A=&MakeFloatFromVector(Vector=$User.ShoeScale_L, Channel=::NewEnumerator0), B=1)
      2=?
      3=?
      Index=&SpawnGroup()
      Num=::NewEnumerator0
     )
     B=&FloatFromCurve(FloatCurve=DataInterface() {propertyValues {Curve {
         defaultValue=0
         keys [
          <at -0.0039 0.7017 RCIM_Cubic>
          {arriveTangent=0.6215, interpMode=RCIM_Cubic, leaveTangent=0.6215, time=0.1727, value=1.7633}
          {arriveTangentWeight=0.1264, interpMode=RCIM_Cubic, leaveTangentWeight=0.2819, tangentMode=RCTM_User, time=0.3923, value=0.948}
          {arriveTangent=-3.2456, interpMode=RCIM_Cubic, leaveTangent=-3.2456, tangentMode=RCTM_User, time=1, value=0}
         ]
        }}})
    )
   )
   &ScaleColor(
    Scale Alpha=&FloatFromCurve(FloatCurve=DataInterface() {propertyValues {Curve {keys [<at 0.5 1 RCIM_Cubic>,<at 1.0011 0.002 RCIM_Cubic>]}}}, Scale Curve=0.5)
    Scale RGB=&MakeVectorFromLinearColorRGB(Color=&FN_Select_Color_from_Time_Of_Day(
      Morning=<rgba 0.271 0.4832 0.8003 1>
      Day=<rgba 0.271 0.4832 0.8003 1>
      Evening=<rgba 0.271 0.4832 0.8003 1>
      Night=<rgba 0.2107 0.4834 0.9219 1>
     ))
    ScaleA=true
    ScaleRGB=true
   )
   &Drag(Use Linear Drag=true, Use Rotational Drag=false)
   &GravityForce(Gravity=&RandomRangeVector(Minimum=<0 0 -300>, Maximum=<0 0 -500>)) {module=GravityForce001}
   &SolveForcesAndVelocity(Clamp Velocity=false, Limit Acceleration=false)
  ]
  rendererClasses [&NiagaraSpriteRendererProperties]
  renderers [{rendererClass=&NiagaraSpriteRendererProperties}]
 }
 Smoke {
  emitterUpdateScript [
   &EmitterState()
   &SpawnRate(
    Spawn Probability=&MakeCustomFloatFromBool(Bool=$User.IsSkydiving, True Float=0, False Float=1)
    SpawnRate=&MakeCustomFloatFromBool(
     Bool=&Boolean_OrOperation(Bool1=$User.IsSkydiving, Bool2=&Boolean_NotOperation(Bool=$User.isStage3))
     True Float=0
     False Float=&MakeCustomFloatFromBool(
      Bool=$User.HideFootFX
      True Float=0
      False Float=&MakeCustomFloatFromBool(Bool=$System.isRunning, True Float=100, False Float=0)
     )
    )
    Use Spawn Probability=true
   )
  ]
  particleSpawnScript [
   &InitializeParticle(
    Color=&MakeCustomLinearColorFromBool(
     Bool=&Boolean_OrOperation(Bool1=$User.bIsFrontEnd, Bool2=$User.bIsFrontEndPreview)
     True Color=<rgba 1 1 1 1>
     False Color=<rgba 1 1 1 0.05>
    )
    Lifetime=&FN_FrontEndFloatModifier(Gameplay=0.3, Front End=0.85)
    Sprite Rotation Mode=::NewEnumerator2
    Sprite Size Mode=::NewEnumerator3
    Uniform Sprite Size=&UniformRangedFloat(Minimum=37, Maximum=52)
    UsePositionOffset=false
    Write Material Random=true
   ) {module=InitializeParticle001}
   &FN_SetSpawnGroup(Number of Spawn Groups=2)
   &SkeletalMeshLocation(Bone / Socket Index=$Particles.SpawnGroup, Bone Sampling Mode=::NewEnumerator4, MaskSpawnGroup=false, Skeletal Mesh=$System.Skeletal Mesh) {module=SkeletalMeshLocation001}
   /SetVariables(Particles.InitSpriteRotRate=&UniformRangedFloat(Minimum=-1, Maximum=1)) {isSetParameters=true}
   &AddVelocity(
    Additional Quaternion Rotation=true
    Additional Yaw / Pitch / Roll:Vector3f=&MaskVectorBySpawnGroup(Masked Vector=<0 0 180>, Passthrough Vector=<0 0 0>)
    Rotation Mode=::NewEnumerator3
    Rotation Quaternion:Quat4f=$Particles.SkeletalMeshLocation001.SampledRotation
    Velocity=<0 120 30>
    Velocity Speed Scale=&FN_FrontEndFloatModifier(Gameplay=0, Front End=1)
   ) {module=AddVelocity001}
   &ApplyInitialForces()
  ]
  particleUpdateScript [
   &ParticleState()
   &SubUVAnimation(
    End Frame Range Override=10
    "Sprite Renderer "=DataInterface() {propertyValues {SpriteRenderer=&self:Smoke_1.NiagaraSpriteRendererProperties_0}}
    UseEndFrame=true
    UseStartFrame=false
   )
   &SpriteRotationRate(Rotation Rate=&Multiply_Float(
     A=$Particles.InitSpriteRotRate
     B=&FloatFromCurve(
      FloatCurve=DataInterface() {propertyValues {Curve {keys [
          {arriveTangent=-1.9214, interpMode=RCIM_Cubic, leaveTangent=-1.9214, tangentMode=RCTM_User, time=-0.0046, value=1.0113}
          {arriveTangent=-0.2496, interpMode=RCIM_Cubic, leaveTangent=-0.2496, tangentMode=RCTM_User, time=0.9896, value=0.16}
         ]}}}
      CurveIndex=$Emitter.Age
     )
    ))
   &ScaleColor(
    Scale Alpha=&FloatFromCurve(
     FloatCurve=DataInterface() {propertyValues {Curve {
        defaultValue=0
        keys [
         {interpMode=RCIM_Cubic, tangentMode=RCTM_User, time=0, value=0}
         {arriveTangent=-2.9824, interpMode=RCIM_Cubic, leaveTangent=-2.9824, tangentMode=RCTM_User, time=0.2, value=0.975}
         {tangentMode=RCTM_User, time=1, value=0}
        ]
       }}}
     Scale Curve=&Multiply_Float(A=$User.TrailOn, B=&FNTimeOfDayScalar(Night Scale=0.5, Morning Scale=1, Evening Scale=1))
    )
    Scale RGB=&MakeVectorFromLinearColorRGB(Color=&ColorFromCurve(ColorCurve=DataInterface() {propertyValues {
        AlphaCurve {keys [<at 0 1 RCIM_Cubic>,<at 1 0 RCIM_Cubic>]}
        BlueCurve {keys [<at 0 0.6094 RCIM_Cubic>,<at 1 0.6962 RCIM_Cubic>]}
        GreenCurve {keys [<at 0 0.3174 RCIM_Cubic>,<at 1 0.526 RCIM_Cubic>]}
        RedCurve {keys [<at 0 0.1923 RCIM_Cubic>,<at 1 0.4371 RCIM_Cubic>]}
       }}))
    ScaleA=true
    ScaleRGB=true
   ) {module=ScaleColor001}
   &Drag(Use Linear Drag=true, Use Rotational Drag=false)
   &InheritSourceMovement(
    Applied Position Scale=&MakeVector(
     X=&FloatFromCurve(FloatCurve=DataInterface() {propertyValues {Curve {
         defaultValue=1
         keys [
          {arriveTangent=-0.8337, interpMode=RCIM_Cubic, leaveTangent=-0.8337, tangentMode=RCTM_User, time=0, value=0.7}
          {arriveTangent=-0.2374, interpMode=RCIM_Cubic, leaveTangent=-0.2374, tangentMode=RCTM_User, time=1, value=0.3}
         ]
        }}})
     Y=&FloatFromCurve(FloatCurve=DataInterface() {propertyValues {Curve {
         defaultValue=1
         keys [
          {arriveTangent=-0.5997, interpMode=RCIM_Cubic, leaveTangent=-0.5997, tangentMode=RCTM_User, time=0, value=0.7}
          {arriveTangent=-0.5578, interpMode=RCIM_Cubic, leaveTangent=-0.5578, tangentMode=RCTM_User, time=1, value=0.3}
         ]
        }}})
     Z=0.2
    )
    ApplyForce=false
    ApplyPosition=true
    ApplyPositionFalloff=false
   )
   &SolveForcesAndVelocity(Clamp Velocity=false, Limit Acceleration=false)
   &ScaleSpriteSize(Scale Factor=&Vector2DFromFloat(Value=&Multiply_Float(
      A=&MergeHelper_FloatSwitch(
       0=&Multiply_Float(A=&MakeFloatFromVector(Vector=$User.ShoeScale_R, Channel=::NewEnumerator0), B=1)
       1=&Multiply_Float(A=&MakeFloatFromVector(Vector=$User.ShoeScale_L, Channel=::NewEnumerator0), B=1)
       2=?
       3=?
       Index=&SpawnGroup()
       Num=::NewEnumerator0
      )
      B=&FloatFromCurve(FloatCurve=DataInterface() {propertyValues {Curve {
          defaultValue=1.4923
          keys [
           {arriveTangent=0.8474, interpMode=RCIM_Cubic, leaveTangent=0.8474, tangentMode=RCTM_User, time=0, value=0.45}
           {interpMode=RCIM_Cubic, tangentMode=RCTM_User, time=1, value=1}
          ]
         }}})
     )))
   &DynamicMaterialParameters(
    Index 0 Param 1=3
    Index 0 Param 2=1
    Index 0 Param 3=0.5
    Module.Index0Float0Write=true
    Module.Index0Float1Write=true
    Module.Index0Float2Write=true
    Module.Index0Float3Write=true
   )
  ]
  rendererClasses [&NiagaraSpriteRendererProperties]
  renderers [{rendererClass=&NiagaraSpriteRendererProperties}]
 }
 Stage3Burst {
  emitterUpdateScript [
   &EmitterState(Life Cycle Mode=::NewEnumerator1, Loop Behavior=::NewEnumerator1, Loop Duration Mode=::NewEnumerator1, UseLoopDelay=false)
   &DoOnce(Trigger Condition:NiagaraBool=$User.isStage3)
   &self:TriggerSpawn(
    Spawn Count=22
    Spawn Group=1
    Trigger:NiagaraBool=&Boolean_OrOperation(Bool1=$Emitter.DoOnce.Execute, Bool2=&Boolean_AndOperation(Bool1=$User.isOffPawnPreview, Bool2=$System.FELoop2))
   ) {module=TriggerSpawn}
  ]
  particleSpawnScript [
   &FN_SetSpawnGroup(Particles Per Group=11, Pattern=::NewEnumerator0)
   &PartitionParticles()
   /SetVariables(Particles.VisibilityTag=$Particles.PartitionParticles.Partition) {isSetParameters=true}
   &InitializeParticle(
    Lifetime=&MaskFloat(
     Mask Source=1
     Mask Parameter=$Particles.PartitionParticles.Partition
     Mask=0
     Masked Float=&RandomRangeFloat(Minimum=0.3, Maximum=0.4)
     Passthrough Float=&RandomRangeFloat(Minimum=0.5, Maximum=1)
    )
    Mesh Scale=&Multiply_VectorByFloat(Vector=&RandomRangeVector(Minimum=<0.15 0.2 0.14>, Maximum=<0.5 0.4 0.3>), Float=0.6)
    Mesh Scale Mode=::NewEnumerator4
    Sprite Rotation Mode=::NewEnumerator2
    Sprite Size Mode=::NewEnumerator1
    Uniform Sprite Size Max=5
    Uniform Sprite Size Min=10
    UsePositionOffset=false
    Write Material Random=true
   )
   &SkeletalMeshLocation(
    Bone / Socket Index=&SelectIntFromArray(
     Array Sampling Mode=::NewEnumerator1
     Direct Array Index=?
     Int Selection Array=DataInterface() {propertyValues {IntData [1,3]}}
     Random Seed=0
    )
    Bone Sampling Mode=::NewEnumerator4
    MaskSpawnGroup=false
    Orientation Sampling=::NewEnumerator1
    Skeletal Mesh=$System.Skeletal Mesh
   )
   &SkeletalMeshLocation(
    Bone / Socket Index=&SelectIntFromArray(
     Array Sampling Mode=::NewEnumerator1
     Direct Array Index=?
     Int Selection Array=DataInterface() {propertyValues {IntData [0,2]}}
     Random Seed=0
    )
    Bone Sampling Mode=::NewEnumerator4
    MaskSpawnGroup=false
    Orientation Sampling=::NewEnumerator1
    Skeletal Mesh=$System.Skeletal Mesh
   ) {module=SkeletalMeshLocation001}
   /SetVariables(Particles.Position=&MergeHelper_VectorSwitch(
     0=Linked() {linkedVariable {name=Particles.SkeletalMeshLocation.SampledPosition, type=NiagaraPosition}}
     1=Linked() {linkedVariable {name=Particles.SkeletalMeshLocation001.SampledPosition, type=NiagaraPosition}}
     2=?
     3=?
     Index=&SpawnGroup()
     Num=::NewEnumerator0
    )) {isSetParameters=true}
   &ShapeLocation(
    MaskSpawnGroup=false
    Override Seed=false
    Radius Position:NiagaraFloat=&RandomRangeFloat(Minimum=0.6, Maximum=1)
    Ring / Disc Distribution Mode:ENiagaraSphereDistributionMode=::NewEnumerator1
    Ring Radius=&Multiply_Float(A=9, B=1)
    Shape Primitive=::NewEnumerator4
    U Position=&ReturnNormalizedExecIndex()
    UseOwnerScale=true
   ) {module=ShapeLocation001}
   &AddVelocity(Velocity=&RandomRangeVector(Minimum=<0 0 80>, Maximum=<0 0 120>)) {module=AddVelocity001}
   &AddVelocity(
    Module.Use Falloff Distance=false
    Override Seed=false
    Velocity Mode=::NewEnumerator1
    Velocity Origin=&MergeHelper_VectorSwitch(
     0=Linked() {linkedVariable {name=Particles.SkeletalMeshLocation.SampledPosition, type=NiagaraPosition}}
     1=Linked() {linkedVariable {name=Particles.SkeletalMeshLocation001.SampledPosition, type=NiagaraPosition}}
     2=?
     3=?
     Index=&SpawnGroup()
     Num=::NewEnumerator0
    )
    Velocity Speed=&RandomRangeFloat(Minimum=25, Maximum=100)
   ) {module=AddVelocity002}
   &InitialMeshOrientation(Mesh Axis To Orient=<0 0 -1>, Mesh Orientation Mode=::NewEnumerator2, Use Rotation Vector=false)
   &CameraOffset(Camera Offset Amount=10)
  ]
  particleUpdateScript [
   &ParticleState()
   &SpriteRotationRate(Rotation Rate=&RandomRangeFloat(Minimum=-300, Maximum=300))
   &ScaleSpriteSize(
    Scale Sprite Size Mode=::NewEnumerator0
    Uniform Scale Factor=&Multiply_Float(
     A=&MergeHelper_FloatSwitch(
      0=&Multiply_Float(A=&MakeFloatFromVector(Vector=$User.ShoeScale_L, Channel=::NewEnumerator0), B=1)
      1=&Multiply_Float(A=&MakeFloatFromVector(Vector=$User.ShoeScale_R, Channel=::NewEnumerator0), B=1)
      2=?
      3=?
      Index=&SpawnGroup()
      Num=::NewEnumerator0
     )
     B=&FloatFromCurve(FloatCurve=DataInterface() {propertyValues {Curve {
         defaultValue=0
         keys [
          <at -0.0039 0.7017 RCIM_Cubic>
          {arriveTangent=0.4635, interpMode=RCIM_Cubic, leaveTangent=0.4635, time=0.1727, value=1.7633}
          {arriveTangentWeight=0.1264, interpMode=RCIM_Cubic, leaveTangentWeight=0.2819, tangentMode=RCTM_User, time=0.5273, value=0.948}
          {arriveTangent=-3.2456, interpMode=RCIM_Cubic, leaveTangent=-3.2456, tangentMode=RCTM_User, time=1, value=0}
         ]
        }}})
    )
   )
   &ScaleColor(
    Scale Alpha=&FloatFromCurve(
     FloatCurve=DataInterface() {propertyValues {Curve {keys [<at 0.5 1 RCIM_Cubic>,<at 1.0011 0.002 RCIM_Cubic>]}}}
     Scale Curve=&MaskFloat(Mask Source=1, Mask Parameter=$Particles.PartitionParticles.Partition, Mask=0, Masked Float=0.22, Passthrough Float=0.8)
    )
    Scale RGB=&MakeVectorFromLinearColorRGB(Color=&FN_Select_Color_from_Time_Of_Day(
      Morning=<rgba 0.2723 0.3937 0.8929 1>
      Day=<rgba 0.2723 0.3937 0.8929 1>
      Evening=<rgba 0.2723 0.3937 0.8929 1>
      Night=<rgba 0.2503 0.4148 1 1>
     ))
    ScaleA=true
    ScaleRGB=true
   )
   &ScaleMeshSize(Scale Factor=&VectorFromFloat(Value=&Multiply_Float(
      A=&MergeHelper_FloatSwitch(
       0=&Multiply_Float(A=&MakeFloatFromVector(Vector=$User.ShoeScale_L, Channel=::NewEnumerator0), B=1)
       1=&Multiply_Float(A=&MakeFloatFromVector(Vector=$User.ShoeScale_R, Channel=::NewEnumerator0), B=1)
       2=?
       3=?
       Index=&SpawnGroup()
       Num=::NewEnumerator0
      )
      B=&FloatFromCurve(
       FloatCurve=DataInterface() {propertyValues {Curve {keys [
           {arriveTangent=9.0725, interpMode=RCIM_Cubic, leaveTangent=9.0725, tangentMode=RCTM_User, time=0, value=0}
           {arriveTangent=-2.5249, interpMode=RCIM_Cubic, leaveTangent=-2.5249, tangentMode=RCTM_User, time=0.1408, value=0.9847}
           {arriveTangent=-0.0405, interpMode=RCIM_Cubic, leaveTangent=-0.0405, tangentMode=RCTM_User, time=1, value=0}
          ]}}}
       Scale Curve=1.7
      )
     )))
   &Drag(
    Drag=&MaskFloat(Mask Source=1, Mask Parameter=$Particles.PartitionParticles.Partition, Mask=0, Masked Float=1, Passthrough Float=5)
    Use Linear Drag=true
    Use Rotational Drag=true
   )
   &GravityForce(Gravity=&MaskVector(
     Mask Source=1
     Mask Parameter=$Particles.PartitionParticles.Partition
     Mask=0
     Masked Vector=<0 0 -300>
     Passthrough Vector=&RandomRangeVector(Minimum=<0 0 0>, Maximum=<0 0 -50>)
    )) {module=GravityForce001}
   &CurlNoiseForce(
    Bias Noise=false
    Noise Strength=&MaskFloat(Mask Source=1, Mask Parameter=$Particles.PartitionParticles.Partition, Mask=0, Masked Float=0, Passthrough Float=100)
    Pan Noise=false
   )
   &SolveForcesAndVelocity(Clamp Velocity=false, Limit Acceleration=false)
   &ScaleMeshSizeBySpeed(Max Scale Factor=<1 1 4>)
   &DynamicMaterialParameters(
    Index 0 Param 1=&FloatFromCurve(FloatCurve=DataInterface() {propertyValues {Curve {
        defaultValue=1
        keys [
         {arriveTangent=2.1404, interpMode=RCIM_Cubic, leaveTangent=2.1404, tangentMode=RCTM_User, time=0.1723, value=0}
         {interpMode=RCIM_Cubic, tangentMode=RCTM_User, time=1, value=0.7672}
        ]
       }}})
    Module.Index0Float0Write=true
    Module.Index0Float1Write=false
    Module.Index0Float2Write=false
    Module.Index0Float3Write=false
   )
   /SetVariables(Particles.IceColor:LinearColor=&MakeLinearColorFromVectorAndFloat(
     "Vector (RGB)"=&MakeVectorFromLinearColorRGB(Color=&FN_Select_Color_from_Time_Of_Day(Morning=<rgba 0.2745 0.5289 1 1>, Day=<rgba 0.2745 0.5289 1 1>, Evening=<rgba 0.2745 0.5289 1 1>, Night=<rgba 0.1102 0.39 1 1>))
     "Float (Alpha)"=1
    )) {isSetParameters=true}
  ]
  rendererClasses [&NiagaraMeshRendererProperties, &NiagaraSpriteRendererProperties]
  renderers [
   {rendererClass=&NiagaraMeshRendererProperties, rendererIndex=0}
   {rendererClass=&NiagaraSpriteRendererProperties, rendererIndex=1}
  ]
 }
}
summary {
 systemName=NS_Shoe_FrostChill_Idle
 emitters [
  {emitterName=Frost, rendererClasses [&NiagaraSpriteRendererProperties]}
  {emitterName=Stage3Burst, rendererClasses [&NiagaraMeshRendererProperties, &NiagaraSpriteRendererProperties]}
  {emitterName=Smoke, rendererClasses [&NiagaraSpriteRendererProperties]}
  {emitterName=FN_AttachedRibbon001, rendererClasses [&NiagaraRibbonRendererProperties]}
  {emitterName=Ice, rendererClasses [&NiagaraSpriteRendererProperties]}
 ]
 userVariables [
  User.bIsFrontEnd {type=NiagaraBool}
  User.bIsFrontEndPreview {type=NiagaraBool}
  User.bisLit {type=NiagaraBool}
  User.HideFootFX {type=NiagaraBool}
  User.isOffPawnPreview {type=NiagaraBool}
  User.IsSkydiving {type=NiagaraBool}
  User.isStage2 {type=NiagaraBool}
  User.isStage3 {type=NiagaraBool}
  User.RibbonIDOffset {type=NiagaraInt32}
  User.ShoeScale_L {type=Vector3f, description="", defaultValue {struct=Vector3f, value=<1 1 1>}}
  User.ShoeScale_R {type=Vector3f, description="", defaultValue {struct=Vector3f, value=<1 1 1>}}
  User.SkeletalMesh {
   type=NiagaraDataInterfaceSkeletalMesh
   defaultValue {
    struct=NiagaraExt_VariableValue_DataInterface
    value {dataInterface=&self:NiagaraDataInterfaceSkeletalMesh_1}
   }
  }
  User.TrailOn {type=NiagaraFloat, defaultValue=1, description=""}
 ]
}
