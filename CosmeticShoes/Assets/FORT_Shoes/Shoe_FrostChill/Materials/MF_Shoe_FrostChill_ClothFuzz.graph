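# Material function for the FrostChill cosmetic shoe that adds optional cloth fuzz, blended through a selectable mask channel.
# A view-angle Fresnel, modulated by a tiled fuzz texture (constant on low quality and mobile), drives tinted base color.
# A quality switch uses a cheaper color and roughness version below Epic, alongside a Cloth shading model variant with Substrate roughness and specular fixes.
# Bounding-box slider nodes build an A/B preview with red border.
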MaterialFunction /CosmeticShoes/Assets/FORT_Shoes/Shoe_FrostChill/Materials/MF_Shoe_FrostChill_ClothFuzz
graph {nodes [
  FunctionInput() {inputName=Fresnel, inputType=FunctionInput_Scalar, sortPriority=2}
  n1 = NamedRerouteUsage()
  n2 = GetMaterialAttributes(None=FunctionInput() {inputName=Material, inputType=FunctionInput_MaterialAttributes}) {attributeGetTypes [#1, #2, #3]}
  n3 = VectorParameter() {group=06_Cloth, parameterName=ClothFuzzTint, defaultValue=<rgba 1 1 1 1>}
  n4 = ScalarParameter() {defaultValue=0.9, group=06_Cloth, parameterName=Cloth_BaseColorIntensity}
  n5 = PixelNormalWS()
  n6 = CameraVectorWS()
  n7 = Power(Base=OneMinus(None=DotProduct(A=@n5, B=@n6)), Exp=ScalarParameter() {defaultValue=4, group=06_Cloth, parameterName=ClothFuzzExponent})
  n8 = TextureSampleParameter2D(UVs=Multiply(A=TextureCoordinate(), B=ScalarParameter() {defaultValue=4, group=06_Cloth, parameterName=ClothFuzzTiling})) {
   group=06_Cloth
   parameterName=ClothFuzz Texture
   samplerSource=SSM_Wrap_WorldGroupSettings
   samplerType=SAMPLERTYPE_LinearColor
   texture=/Game/Global/Textures/Default/Masks_and_Gradients/T_Fuzz_MASK
  }
  n9 = Constant() {r=0.65}
  n10 = QualitySwitch(Default=@n8.R, Low=@n9)
  n11 = Multiply(
   A=Multiply(A=Clamp(None=@n7), B=ShadingPathSwitch(Default=@n10, Deferred=@n10, Mobile=@n9))
   B=ScalarParameter() {defaultValue=1, group=06_Cloth, parameterName=Fuzz Fresnel Blend}
  )
  n12 = Add(A=Multiply(A=Saturate(None=@n11)) {constB=10})
  n13 = Saturate(None=Multiply(A=Multiply(A=@n2.BaseColor, B=@n4), B=Multiply(A=@n3.RGB, B=@n12)))
  n14 = Add(
   A=LinearInterpolate(
    A=@n2.BaseColor
    B=Multiply(A=@n2.BaseColor, B=@n3.RGB)
    Alpha=ScalarParameter() {group=06_Cloth, parameterName=Standard Quality Cloth Fuzz Tint Influence, sortPriority=34}
   )
   B=Multiply(
    A=ScalarParameter() {defaultValue=1, group=06_Cloth, parameterName=Standard Quality Cloth Fuzz Intensity Multiplier, sortPriority=33}
    B=@n13
   )
  )
  n15 = GetMaterialAttributes(None=NamedRerouteUsage()) {attributeGetTypes [#3]}
  n16 = Multiply(A=ScalarParameter() {defaultValue=1, group=06_Cloth, parameterName=ClothRoughnesss_StandardQuality}, B=@n15.Roughness)
  n17 = NamedRerouteDeclaration(None=SetMaterialAttributes!(MaterialAttributes=@n15.MaterialAttributes, BaseColor=Saturate(None=@n14), Roughness=Saturate(None=@n16)) {attributeSetTypes [#1, #3]}) {name=ClothFuzz_Low, nodeColor=<rgba 0 1 0.9766 1>}
  n18 = MaterialFunctionCall!(
   Low _ PC Low and Med=@n17
   "High _ XB1, PS4"=@n17
   "Epic _ PC High, XSX, PS5"=StaticSwitchParameter(True=NamedRerouteUsage(), False=NamedRerouteUsage()) {group=06_Cloth, parameterName=Preview Standard Quality Cloth Fuzz On Epic Scalability, sortPriority=34}
   --------
   "Mobile Low _ PC Perf, Android"=@n17
   Mobile Med _ Switch=@n17
  ) {materialFunction=/Game/Global/MaterialFunctions/TechArt/MF_Fortnite_QualitySwitch_MaterialAttributes}
  n19 = ChannelMaskParameter(None=FunctionInput() {bUsePreviewValueAsDefault=true, inputName=Alpha, sortPriority=3}) {group=05_Tech Art Masks, parameterName=ClothFuzzMaskChannel}
  FunctionOutput(None=StaticSwitchParameter(True=BlendMaterialAttributes(A=@n1, B=@n18.Result, Alpha=@n19.R), False=@n1) {group=04_Tech Art, parameterName=UseClothFuzz})
  NamedRerouteDeclaration(None=@n2.MaterialAttributes) {name=Material Attributes, nodeColor=<rgba 1 0.4922 0 1>}
  n20 = MaterialFunctionCall() {materialFunction=&BoundingBoxBased_0-1_UVW}
  n21 = ScalarParameter() {group=06_Cloth, parameterName=Preview Standard Quality Cloth Fuzz - A/B Slider, sliderMax=1, sortPriority=35}
  NamedRerouteDeclaration(None=If(A=@n20.R, B=@n21, "A > B"=Constant(), "A < B"=Constant() {r=1})) {name=PreviewClothFuzzSlider, nodeColor=<rgba 0 1 0.5781 1>}
  n22 = ScalarParameter() {defaultValue=1, group=06_Cloth, parameterName=ClothSubsurfaceIntensity}
  n23 = Saturate(None=Multiply(A=@n3.RGB, B=Multiply(A=@n2.BaseColor, B=@n22)))
  n24 = ScalarParameter() {defaultValue=0.4, group=06_Cloth_SubstrateFix, parameterName=SubstrateClothFix_RoughnessBias, sliderMax=1, sortPriority=0}
  n25 = Saturate(None=Multiply(A=@n2.Roughness, B=ScalarParameter() {defaultValue=1, group=06_Cloth, parameterName=ClothRoughness}))
  n26 = ScalarParameter() {defaultValue=5, group=06_Cloth_SubstrateFix, parameterName=SubstrateClothFix_SpecularFresnelExponent, sortPriority=0}
  n27 = ScalarParameter() {defaultValue=0.3, group=06_Cloth_SubstrateFix, parameterName=SubstrateClothFix_SpecularFresnelBase, sortPriority=0}
  n28 = FunctionInput() {inputName=Ao/Gmap/Skin, sortPriority=2}
  n29 = Saturate(None=LinearInterpolate(
    A=ScalarParameter() {defaultValue=0.75, group=06_Cloth, parameterName=ClothScatteringMin}
    B=ScalarParameter() {defaultValue=1, group=06_Cloth, parameterName=ClothScatteringMax}
    Alpha=MaterialFunctionCall(Float3=@n28) {materialFunction=&BreakOutFloat3Components}.G
   ))
  n30 = SetMaterialAttributes!(
   MaterialAttributes=NamedRerouteUsage()
   ShadingModel=ShadingModel() {shadingModel=MSM_Cloth}
   SubsurfaceColor=@n23
   Roughness=LinearInterpolate(A=@n24, Alpha=Saturate(None=@n25))
   CustomData0=ScalarParameter() {defaultValue=1, group=06_Cloth, parameterName=ClothShaderIntensity}
   Opacity=Constant() {r=1}
   BaseColor=@n13
   Specular=Multiply(A=Fresnel(ExponentIn=@n26, BaseReflectFractionIn=@n27) {baseReflectFraction=0}, B=@n29)
  ) {attributeSetTypes [
    #4
    #5
    #3
    #6
    #7
    #1
    #8
   ]}
  NamedRerouteDeclaration(None=@n30) {name=ClothFuzz_High, nodeColor=<rgba 1 0.6563 0 1>}
  NamedRerouteDeclaration(None=SphereMask(A=@n20.R, B=LinearInterpolate(Alpha=@n21) {constA=-0.003, constB=1.003}) {attenuationRadius=0.003}) {name=PreviewClothFuzzSlider_Border, nodeColor=<rgba 0.3359 0 1 1>}
  n31 = BlendMaterialAttributes(A=NamedRerouteUsage(), B=NamedRerouteUsage(), Alpha=NamedRerouteUsage()) {vertexAttributeBlendType=UseA}
  n32 = SetMaterialAttributes!(MaterialAttributes, EmissiveColor=Constant3Vector() {constant=<rgba 1 0 0 0>}) {attributeSetTypes [#9]}
  n33 = NamedRerouteUsage()
  NamedRerouteDeclaration(None=StaticSwitchParameter(True=BlendMaterialAttributes(A=@n31, B=@n32, Alpha=@n33) {vertexAttributeBlendType=UseA}, False=@n31) {defaultValue=true, group=06_Cloth, parameterName=Preview Standard Quality Cloth Fuzz - Show A/B Border, sortPriority=36}) {name=ClothFuzz_LowPreview, nodeColor=<rgba 0 1 0.0156 1>}
  SetMaterialAttributes!(
   MaterialAttributes=NamedRerouteUsage()
   ShadingModel=ShadingModel() {shadingModel=MSM_Cloth}
   SubsurfaceColor=@n23
   Roughness=@n25
   CustomData0=ScalarParameter() {defaultValue=1, group=06_Cloth, parameterName=ClothShaderIntensity}
   Opacity=Constant() {r=1}
   BaseColor=@n13
   Specular=@n29
  ) {attributeSetTypes [
    #4
    #5
    #3
    #6
    #7
    #1
    #8
   ]}
 ]}
