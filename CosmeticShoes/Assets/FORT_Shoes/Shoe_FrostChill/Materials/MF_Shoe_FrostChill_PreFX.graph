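# Material function for the Frost Chill cosmetic shoe that builds the base surface ahead of the FX stage implied by its PreFX name.
# It chains the character parent texture and skin functions, brightness and scale balance, then the shoe's BasicTechArt function, which takes a frost mask.
# UVs use a switchable channel with input overrides.
# It also outputs masks, distance blend, Fresnel and pre-skinned position, interpolated for pixel use.
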MaterialFunction /CosmeticShoes/Assets/FORT_Shoes/Shoe_FrostChill/Materials/MF_Shoe_FrostChill_PreFX
graph {
 properties {bExposeToLibrary=true}
 nodes [
  n1 = FunctionInput(Preview=NamedRerouteUsage()) {
   bUsePreviewValueAsDefault=true
   inputName=Override Tech Art Mask UVs
   inputType=FunctionInput_Vector2
   sortPriority=11
   previewValue=<1 1 0 1>
  }
  n2 = FunctionInput(Preview=NamedRerouteUsage()) {
   bUsePreviewValueAsDefault=true
   inputName=Override Basic Emissive UVs
   inputType=FunctionInput_Vector2
   sortPriority=12
   previewValue=<1 1 0 1>
  }
  n3 = StaticSwitchParameter(True=TextureCoordinate() {coordinateIndex=1}, False=TextureCoordinate()) {group=00_Texture Inputs, parameterName=useSecondUVChannel, sortPriority=40}
  n4 = MaterialFunctionCall!(UV=FunctionInput(Preview=NamedRerouteDeclaration(None=@n3) {name=UVs, nodeColor=<rgba 0.2891 0 1 1>}) {bUsePreviewValueAsDefault=true, inputName=Overide Texture Set UVs, inputType=FunctionInput_Vector2}) {materialFunction=/Game/Global/Materials/Character_Parent_2023/Functions/MF_CharacterParent_2023_TextureInputs}
  n5 = MaterialFunctionCall!(Material=MaterialFunctionCall!(Skin Mask=NamedRerouteUsage(), GMap=NamedRerouteUsage(), Material=@n4.Result) {materialFunction=/Game/Global/Materials/Character_Parent_2023/Functions/MF_CharacterParent_2023_Skin}.Result) {materialFunction=/Game/Packages/Fortress_SharedMaterials/Base_Material_Functions/CharacterBrightnessandScale_Balance}
  n6 = MaterialFunctionCall!(
   Tech Art Mask UVs=@n1
   FrostMask=FunctionInput() {inputName=FrostMask, inputType=FunctionInput_Scalar}
   Emissive UVs=@n2
   LocalPosition=NamedRerouteUsage()
   Fresnel=NamedRerouteUsage()
   Mask Input=NamedRerouteUsage()
   Material=@n5.Result
  ) {materialFunction=&/MF_Shoe_FrostChill_BasicTechArt}
  FunctionOutput(None=@n6.Result)
  FunctionOutput(None=@n5.Distance Blend) {outputName=Distance Blend, sortPriority=4}
  FunctionOutput(None=@n4.M) {outputName=M, sortPriority=1}
  n7 = Constant() {r=1}
  FunctionOutput(None=NamedRerouteDeclaration(None=Fresnel(ExponentIn=@n7) {baseReflectFraction=0}) {name=Fresnel, nodeColor=<rgba 1 0.3047 0 1>}) {outputName=Fresnel, sortPriority=9}
  FunctionOutput(None=@n6.Tech Art Mask) {outputName=Tech Art Mask, sortPriority=11}
  FunctionOutput(None=@n6.Tech Art Mask Alpha) {outputName=Tech Art Mask Alpha, sortPriority=12}
  n8 = PreSkinnedPosition()
  n9 = VertexInterpolator(VS=@n8)
  FunctionOutput(None=NamedRerouteDeclaration(None=ShaderStageSwitch(PixelShader=@n9.PS, VertexShader=@n8)) {name=Pre Skinned Local Postition, nodeColor=<rgba 1 0.5156 0 1>}) {outputName=Pre skinned Local Position, sortPriority=2}
  n10 = MaterialFunctionCall(Float3=@n4.M) {desc="R = AO\r\nG = GMap\r\nB = Skin", materialFunction=&BreakOutFloat3Components}
  NamedRerouteDeclaration(None=@n10.R) {name=Ambient Occlusion, nodeColor=<rgba 0.875 0 1 1>}
  NamedRerouteDeclaration(None=@n10.G) {name=GMap, nodeColor=<rgba 0.6641 0 1 1>}
  NamedRerouteDeclaration(None=@n10.B) {name=Skin, nodeColor=<rgba 0.6641 0 1 1>}
  NamedRerouteDeclaration(None=GetMaterialAttributes(None=@n4.Result) {attributeGetTypes [#1]}.Normal) {name=Normal Map, nodeColor=<rgba 1 0.3281 0 1>}
  NamedRerouteDeclaration(None=@n4.M) {name=Mask, nodeColor=<rgba 0.3125 0 1 1>}
 ]
}
