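# Parent material for the FrostChill cosmetic shoe, whose frost and ice-mesh effects react to kill count.
# KillCount animates over a timestamped duration, or cycles on time when IsOffPawn is set.
# A local-height gradient plus tiled noise forms the frost mask, driving ice color, normals and MF_FrostChill_Ice, with time-of-day-scaled edge glow.
# Vertex-color masks isolate sole and ice mesh, which uses opacity and collapse-to-pivot WPO, before shared shoe functions.
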Material /CosmeticShoes/Assets/FORT_Shoes/Shoe_FrostChill/Materials/M_FrostChill_Parent
graph {
 properties {
  bUseMaterialAttributes=true
  bUsedWithClothing=true
  bUsedWithMeshDeformer=true
  bUsedWithMorphTargets=true
  bUsedWithNanite=true
  bUsedWithNiagaraMeshParticles=true
  bUsedWithSkeletalMesh=true
  pixelDepthOffsetMode=PDOM_Legacy
  shadingModel=MSM_FromMaterialExpression
  twoSided=true
  usedShadingModels=Default Lit | Subsurface | Clear Coat | Subsurface Profile | Cloth
 }
 nodes [
  n1 = VertexInterpolator(VS=LocalPosition() {localOrigin=InstancePreSkinning}.XYZ)
  n2 = ComponentMask(None=@n1.PS) {b=true}
  n3 = Saturate(None=Divide(A=NamedRerouteUsage()) {constB=1.5})
  n4 = Multiply(A=ComponentMask(None=@n1.PS) {g=true}, B=Multiply(A=Constant() {r=0.15}, B=@n3))
  n5 = ScalarParameter() {defaultValue=10, group=FrostChill, parameterName=Frost Height}
  n6 = Constant() {r=8}
  n7 = Power(Base=@n3, Exp=Constant() {r=0.7}) {constExponent=0.5}
  n8 = Add(
   A=NamedRerouteDeclaration(None=Add(A=@n2, B=@n4)) {name=HeightGradient, nodeColor=<rgba 0 0.1328 1 1>}
   B=Multiply(A=Divide(A=@n5, B=@n6) {constB=5}, B=OneMinus(None=@n7))
  )
  n9 = Multiply(A=@n5, B=LinearInterpolate(A=Divide(B=@n6) {constB=5}, Alpha=@n7) {constA=0.3})
  n10 = OneMinus(None=Saturate(None=Divide(A=@n8, B=@n9)))
  n11 = Saturate(None=NamedRerouteUsage())
  Add(A=@n10, B=Multiply(A=NamedRerouteUsage(), B=Power(Base=@n11)))
  n12 = TextureSample(UVs=Multiply(A=TextureCoordinate(), B=ScalarParameter() {defaultValue=4, group=FrostChill, parameterName=FrostTile})) {
   samplerSource=SSM_Wrap_WorldGroupSettings
   samplerType=SAMPLERTYPE_Masks
   texture=/CosmeticShoes/Assets/FORT_Shoes/Shoe_FrostChill/Textures/T_FrostNoise
  }
  n13 = Saturate(None=Multiply(A=NamedRerouteUsage(), B=Constant() {r=8}))
  n14 = MaterialFunctionCall!(Overide Texture Set UVs, FrostMask=NamedRerouteUsage(), Override Tech Art Mask UVs, Override Basic Emissive UVs) {materialFunction=&/MF_Shoe_FrostChill_PreFX}
  n15 = GetMaterialAttributes(None=@n14.Result) {attributeGetTypes [#1, #2]}
  n16 = TextureSample() {
   samplerSource=SSM_Wrap_WorldGroupSettings
   samplerType=SAMPLERTYPE_Normal
   texture=/CosmeticShoes/Assets/FORT_Shoes/Shoe_FrostChill/Textures/T_FrostChill_Shoe_N_Extra
  }
  n17 = ComponentMask(None=@n14.Tech Art Mask) {b=true}
  n18 = VertexColor()
  n19 = Ceil(None=Subtract(A=@n18.R) {constB=0.1})
  n20 = NamedRerouteDeclaration(None=@n19) {name=Sole Mask, nodeColor=<rgba 1 0 0.3984 1>}
  n21 = Constant() {r=0.5}
  n22 = Saturate(None=Add(A=ConstantBiasScale(None=@n10) {bias=-0.5, scale=2}, B=Power(Base=@n12.R, Exp=@n21)))
  n23 = Max(A=NamedRerouteDeclaration(None=Subtract(A=@n17, B=@n20)) {name=IceMask, nodeColor=<rgba 0.4766 0 1 1>}, B=@n22)
  n24 = SetMaterialAttributes!(
   MaterialAttributes=@n15.MaterialAttributes
   Normal=LinearInterpolate(A=@n15.Normal, B=@n16.RGB, Alpha=@n23)
   Base Color=LinearInterpolate(A=@n15.Base Color, B=VectorParameter() {parameterName=IceBC_Override, defaultValue=<rgba 0.0209 0.1391 0.0682 1>}.RGB, Alpha=@n23)
  ) {attributeSetTypes [#1, #2]}
  n25 = GetMaterialAttributes(None=MaterialFunctionCall!(
    FrostMask=@n23
    FrostNoise=NamedRerouteDeclaration(None=@n12.R) {name=FrostTile, nodeColor=<rgba 0.4531 1 0 1>}
    HighlightMask=OneMinus(None=@n13)
    In=MaterialFunctionCall!(
     Low _ PC Low and Med=@n24
     "High _ XB1, PS4"=@n24
     "Epic _ PC High, XSX, PS5"=@n24
     --------
     "Mobile Low _ PC Perf, Android"=@n15.MaterialAttributes
     Mobile Med _ Switch=@n24
    ) {materialFunction=/Game/Global/MaterialFunctions/TechArt/MF_Fortnite_QualitySwitch_MaterialAttributes}.Result
    InMask=@n14.M
    InFXMask=@n23
   ) {materialFunction=&/MF_FrostChill_Ice}.Result) {attributeGetTypes [#3, #4]}
  n26 = Divide(A=ScalarParameter() {parameterName=HUE Shift}) {constB=359}
  n27 = ScalarParameter() {defaultValue=1, parameterName=Saturation}
  n28 = Multiply(
   A=Desaturation(
    None=MaterialFunctionCall(Hue Shift Percentage=@n26, Texture=@n25.Emissive Color) {materialFunction=&HueShift}.Result
    Fraction=OneMinus(None=@n27)
   )
   B=ScalarParameter() {defaultValue=1, parameterName=Value}
  )
  n29 = SetMaterialAttributes!(
   MaterialAttributes=@n25.MaterialAttributes
   World Position Offset=NamedRerouteUsage()
   Opacity Mask=NamedRerouteUsage()
   Emissive Color=Add(A=LinearInterpolate(A=@n25.Emissive Color, B=@n28, Alpha=@n23), B=NamedRerouteUsage())
   Roughness=Add(A=@n25.Roughness, B=@n13)
  ) {attributeSetTypes [
    #5
    #6
    #3
    #4
   ]}
  n30 = MaterialFunctionCall!(Material=@n29) {materialFunction=/Game/Global/Materials/Cosmetics/Kicks/Shoes/Materials/MF_ShoeReactivityTest}
  MaterialFunctionCall!(In=MaterialFunctionCall(
    Material=MaterialFunctionCall!(InOpacity, Material=@n30.Result, Mask=@n14.Tech Art Mask) {materialFunction=/Game/Global/Materials/Cosmetics/Kicks/Socks/Materials/MF_Sock}.Result
    M=@n14.M
    Pre Skinned Local Position=@n14.Pre skinned Local Position
    Distance Blend=@n14.Pre skinned Local Position
    Fresnel=@n14.Pre skinned Local Position
   ) {materialFunction=&MF_CharacterParent_2023_PostFX}.Result) {materialFunction=/Game/Global/Materials/Cosmetics/Kicks/Shoes/Materials/MF_TechartShoeHologram}
  Max(A=NamedRerouteUsage(), B=NamedRerouteUsage())
  n31 = Ceil(None=@n18.R)
  NamedRerouteDeclaration(None=Subtract(A=Multiply(A=@n31), B=@n19)) {name=IceMesh Mask, nodeColor=<rgba 0.0781 1 0 1>}
  n32 = NamedRerouteUsage()
  n33 = Constant() {r=10}
  n34 = OneMinus(None=Saturate(None=Divide(A=@n32, B=@n33) {constB=15}))
  n35 = Subtract(A=NamedRerouteUsage())
  n36 = ConstantBiasScale(None=Add(A=Saturate(None=@n35), B=ScalarParameter() {defaultValue=1, group=FrostChill, parameterName=IceMesh On Override})) {bias=-0.5, scale=2}
  n37 = ConstantBiasScale(None=Saturate(None=Add(A=@n34, B=@n36))) {bias=-0.5, scale=2}
  n38 = NamedRerouteUsage()
  n39 = SmoothStep(Value=Saturate(None=Add(A=@n37, B=@n38))) {constMin=0.3}
  NamedRerouteDeclaration(None=@n39) {name=IceMeshOpacity, nodeColor=<rgba 1 0 0.1875 1>}
  n40 = NamedRerouteUsage()
  n41 = Constant() {r=0.2}
  n42 = Divide(A=Subtract(A=@n40, B=@n41), B=Subtract(B=@n41))
  n43 = NamedRerouteUsage()
  n44 = Ceil(None=Saturate(None=Subtract(A=@n43)))
  n45 = Multiply(
   A=Multiply(A=Subtract(A=@n40, B=@n42), B=NamedRerouteUsage())
   B=Multiply(A=NamedRerouteUsage(), B=NamedRerouteDeclaration(None=@n44) {name=IsThirdTransition, nodeColor=<rgba 0 1 0.1328 1>})
  )
  n46 = NamedRerouteUsage()
  n47 = Constant() {r=0.5}
  n48 = Subtract(A=@n46, B=Divide(A=Subtract(A=@n46, B=@n47), B=Subtract(B=@n47)))
  n49 = NamedRerouteUsage()
  n50 = Subtract(A=Power(Base=Multiply(A=@n48, B=@n49), Exp=Constant() {r=2.5}), B=NamedRerouteUsage())
  n51 = NamedRerouteDeclaration(None=Add(A=Multiply(A=@n45) {constB=0.5}, B=Saturate(None=@n50))) {name=EdgeMask, nodeColor=<rgba 0 0.0391 1 1>}
  n52 = VectorParameter() {parameterName=FrostLineColor, defaultValue=<rgba 2 4.7957 5 1>}
  n53 = Constant() {r=0.8}
  n54 = Constant() {r=1}
  n55 = Constant() {r=0.6}
  NamedRerouteDeclaration(None=Multiply(
    A=Multiply(A=@n51, B=@n52.RGB)
    B=MaterialFunctionCall(Color_Morning=@n53, Color_Day=@n54, Color_Evening=@n53, Color_Night=@n55) {materialFunction=&Color_Time_of_Day}.Result
   )) {name=GlowEdge, nodeColor=<rgba 0 0.5078 1 1>}
  n56 = ObjectPositionWS()
  n57 = WorldPosition()
  n58 = NamedRerouteUsage()
  n59 = Add(A=Saturate(None=Subtract(A=@n58)), B=ScalarParameter() {defaultValue=1, group=FrostChill, parameterName=IceMesh On Override})
  n60 = Multiply(A=NamedRerouteUsage(), B=OneMinus(None=Ceil(None=@n59)))
  n61 = Multiply(A=NamedRerouteUsage(), B=Floor(None=@n59))
  NamedRerouteDeclaration(None=Multiply(A=Subtract(A=@n56, B=@n57.XYZ), B=Add(A=@n60, B=@n61))) {name=WPO, nodeColor=<rgba 1 0 0.8672 1>}
  NamedRerouteDeclaration(None=@n22) {name=FrostMask, nodeColor=<rgba 0.5 0 1 1>}
  n62 = ScalarParameter() {group=Reactivity, parameterName=KillCount}
  n63 = Subtract(A=@n62)
  n64 = Time()
  n65 = ScalarParameter() {defaultValue=-100000, group=Reactivity, parameterName=Timestamp}
  n66 = Saturate(None=Divide(A=Subtract(A=@n64, B=@n65), B=ScalarParameter() {defaultValue=4, group=Reactivity, parameterName=Reactivity Duration}))
  n67 = Time()
  n68 = ScalarParameter() {defaultValue=4, group=Reactivity, parameterName=Reactivity Duration}
  n69 = Add(A=@n68, B=Constant() {r=2}) {constB=2}
  n70 = Subtract(A=Modulo!(A=Divide(A=@n67, B=@n69), B=Constant() {r=3}))
  n71 = Frac(None=@n70)
  n72 = Subtract(A=@n70, B=@n71)
  n73 = Saturate(None=MaterialFunctionCall(Input=@n71, Target High=Divide(A=@n69, B=@n68)) {materialFunction=&RemapValueRange}.Result)
  NamedRerouteDeclaration(None=LinearInterpolate(A=LinearInterpolate(A=@n63, B=@n62, Alpha=@n66), B=Add(A=@n72, B=@n73), Alpha=ScalarParameter() {parameterName=IsOffPawn})) {name=KillCount, nodeColor=<rgba 0.0781 0 1 1>}
  NamedRerouteDeclaration(None=LinearInterpolate(B=@n39, Alpha=NamedRerouteUsage()) {constA=1, constB=0}) {name=Opacity, nodeColor=<rgba 1 0 0.7734 1>}
  n74 = Frac(None=NamedRerouteUsage())
  n75 = Abs(None=ConstantBiasScale(None=@n74) {bias=-0.5, scale=2})
  n76 = Constant() {r=0.1}
  n77 = Constant() {r=0.5}
  n78 = Round(None=@n74)
  NamedRerouteDeclaration(None=Power(Base=OneMinus(None=@n75), Exp=LinearInterpolate(A=@n76, B=@n77, Alpha=@n78) {constA=5})) {name=TransitionCurve, nodeColor=<rgba 0.4531 0 1 1>}
 ]
}
